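# StateTree for Minor Spire SubScenario A, an encounter objective to destroy three shockwave spire spikes ("Slag").
# States spawn the spikes sequentially, each anchored by a two-tag location query, then pass through Grant Quest, whose grant task is disabled.
# A verb listener for destroyed Slag props loops through an increment state until the count equals the maximum, alongside a global objective-tracking task.
# Completion sends the shield-interact custom verb for each player and succeeds.
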StateTree /SpiresGameplay/Scenarios/Spire_Minor/Master_Scenario/SubScenarios/Spire_Minor_SubScenario_A/ST_SubScenario_A
global_tasks [
 {
  id=#1
  description="Get Variable [Int]"
  properties {
   iD=#2
   instance {
    _structType=/Script/EncountersRuntime.EncounterGetVariableIntStateTreeTaskInstanceData
    userActor=null
    value=0
    variableIdentifier {tagName=QuestEncounter.Spire_Minor_Combat_1.Encounter.Variable.EnemyCount}
   }
   node {
    name=None
    _structType=/Script/EncountersRuntime.EncounterGetVariableIntStateTreeTask
    bCanEditConsideredForCompletion=true
    bConsideredForCompletion=true
    bHasTaskCompletionDelegateDispatcher=false
    bTaskEnabled=true
    transitionHandlingPriority=Normal
    bindingsBatch {value=65535}
    executionRuntimeTemplateIndex {value=65535}
    instanceDataHandle {index=65535, source=None, stateHandle {index=65535}}
    instanceTemplateIndex {value=65535}
    outputBindingsBatch {value=65535}
   }
  }
 }
 {
  id=#3
  description=Get Scenario Actor - MinorSpire
  properties {
   iD=#4
   instance {
    _structType=&GetEncounterActorStateTreeTaskInstanceData
    actorQuery {
     autoDescription=" ANY( QuestEncounter.Spire_Minor_Random.Encounter.Target.Spire )"
     tagDictionary [{tagName=QuestEncounter.Spire_Minor_Random.Encounter.Target.Spire}]
    }
   }
   node {name=Get Scenario Actor - MinorSpire, _structType=&GetEncounterActorStateTreeTask}
  }
 }
 {
  id=#5
  description=Make Int - Current Objective Count
  properties {
   iD=#6
   instance {_structType=/Script/FortniteAIServer.FortAthenaMakeIntVariableStateTreeTaskInstanceData, defaultIntValue=0, outInt {int=0}}
   node {
    name=Make Int - Current Objective Count
    _structType=/Script/FortniteAIServer.FortAthenaMakeIntVariableStateTreeTask
    bCanEditConsideredForCompletion=true
    bConsideredForCompletion=true
    bHasTaskCompletionDelegateDispatcher=false
    bResetOnReselect=false
    bTaskEnabled=true
    transitionHandlingPriority=Normal
    bindingsBatch {value=65535}
    executionRuntimeTemplateIndex {value=65535}
    instanceDataHandle {index=65535, source=None, stateHandle {index=65535}}
    instanceTemplateIndex {value=65535}
    outputBindingsBatch {value=65535}
   }
  }
 }
 {
  id=#7
  description=Make Int - Max Objective Count
  properties {
   iD=#8
   instance {_structType=&FortAthenaMakeIntVariableStateTreeTaskInstanceData, defaultIntValue=3, outInt {int=0}}
   node {
    name=Make Int - Max Objective Count
    _structType=&FortAthenaMakeIntVariableStateTreeTask
    bCanEditConsideredForCompletion=true
    bConsideredForCompletion=true
    bHasTaskCompletionDelegateDispatcher=false
    bResetOnReselect=false
    bTaskEnabled=true
    transitionHandlingPriority=Normal
    bindingsBatch {value=65535}
    executionRuntimeTemplateIndex {value=65535}
    instanceDataHandle {index=65535, source=None, stateHandle {index=65535}}
    instanceTemplateIndex {value=65535}
    outputBindingsBatch {value=65535}
   }
  }
 }
 {
  id=#9
  description=Minor Spire - Objective - Destroyed Spikes Count
  properties {
   iD=#10
   instanceObject=&self:StateTreeEditorData_0.STT_MinorSpire_TrackObjectiveCount_C_0
   node {
    name=Minor Spire - Objective - Destroyed Spikes Count
    _structType=&StateTreeBlueprintTaskWrapper
    taskClass=/SpiresGameplay/Scenarios/Spire_Minor/Master_Scenario/Shared/STT_MinorSpire_TrackObjectiveCount.STT_MinorSpire_TrackObjectiveCount_C
   }
  }
 }
]
states [
 Root {
  id=#11
  type=STATE
  children [
   #12
   #13
   #14
   #15
   #16
   #17
   #18
   #19
  ]
  tasks []
  transitions []
 }
 Init {
  id=#12
  type=STATE
  tasks [{
    id=#20
    description=BP SDA STTask Complete State
    properties {
     iD=#21
     instanceObject=&self:StateTreeEditorData_0.StateTreeState_1.BP_SDA_STTask_CompleteState_C_0
     node {
      name=None
      _structType=&StateTreeBlueprintTaskWrapper
      taskClass=/StateDrivenActor/Tasks/BP_SDA_STTask_CompleteState.BP_SDA_STTask_CompleteState_C
     }
    }
   }]
  transitions [{id=#22, trigger=ON_STATE_COMPLETED, state {id=#23, name=None, linkType=NEXT_STATE}}]
 }
 Spawn Slag 1 {
  id=#13
  type=STATE
  tasks [{
    id=#24
    description=Spawn Actor
    properties {
     iD=#25
     instance {
      _structType=/Script/EncountersRuntime.SpawnActorStateTreeTaskInstanceData
      actorClass=/SpireSharedAssets/GameplayActors/B_Prop_SpireSpike_ShockWave_MinorSpire.B_Prop_SpireSpike_ShockWave_MinorSpire_C
      bAttachEncounterVolume=false
      bEnablePhysics=false
      bUseActorLocation=false
      bWaitForSpawnedActor=true
      owningActor=null
      persistenceBehavior=Destroy
      spawnedActor=null
      targetActor=null
      userActor=null
      actorIdentifierTags {gameplayTags [{tagName=QuestEncounter.Spire_Minor_Combat_1.Location.Object.1}]}
      actorTags {gameplayTags []}
      anchorOffset=<0 0 0>
      anchorQuery {
       autoDescription=" ANY( QuestEncounter.Spire_Minor_Combat_1.Location.Object.1, QuestEncounter.Spire_Minor_Combat_1.Location.Object.2 )"
       tokenStreamVersion=0
       userDescription=""
       queryTokenStream [0,1,1,2,0,1]
       tagDictionary [{tagName=QuestEncounter.Spire_Minor_Combat_1.Location.Object.1}, {tagName=QuestEncounter.Spire_Minor_Combat_1.Location.Object.2}]
      }
      existingActorQuery {autoDescription="", tokenStreamVersion=0, userDescription="", queryTokenStream [], tagDictionary []}
     }
     node {
      name=None
      _structType=/Script/EncountersRuntime.SpawnActorStateTreeTask
      bCanEditConsideredForCompletion=true
      bConsideredForCompletion=true
      bHasTaskCompletionDelegateDispatcher=false
      bTaskEnabled=true
      transitionHandlingPriority=Normal
      bindingsBatch {value=65535}
      executionRuntimeTemplateIndex {value=65535}
      instanceDataHandle {index=65535, source=None, stateHandle {index=65535}}
      instanceTemplateIndex {value=65535}
      outputBindingsBatch {value=65535}
     }
    }
   }]
  transitions [{id=#26, trigger=ON_STATE_COMPLETED, state {id=#23, name=Root, linkType=NEXT_STATE}}]
 }
 Spawn Slag 2 {
  id=#14
  type=STATE
  tasks [{
    id=#27
    description=Spawn Actor
    properties {
     iD=#28
     instance {
      _structType=&SpawnActorStateTreeTaskInstanceData
      actorClass=&B_Prop_SpireSpike_ShockWave_MinorSpire_C
      bAttachEncounterVolume=false
      bEnablePhysics=false
      bUseActorLocation=false
      bWaitForSpawnedActor=true
      owningActor=null
      persistenceBehavior=Destroy
      spawnedActor=null
      targetActor=null
      userActor=null
      actorIdentifierTags {gameplayTags [{tagName=QuestEncounter.Spire_Minor_Combat_1.Location.Object.2}]}
      actorTags {gameplayTags []}
      anchorOffset=<0 0 0>
      anchorQuery {
       autoDescription=" ANY( QuestEncounter.Spire_Minor_Combat_1.Location.Object.3, QuestEncounter.Spire_Minor_Combat_1.Location.Object.4 )"
       tokenStreamVersion=0
       userDescription=""
       queryTokenStream [0,1,1,2,0,1]
       tagDictionary [{tagName=QuestEncounter.Spire_Minor_Combat_1.Location.Object.3}, {tagName=QuestEncounter.Spire_Minor_Combat_1.Location.Object.4}]
      }
      existingActorQuery {autoDescription="", tokenStreamVersion=0, userDescription="", queryTokenStream [], tagDictionary []}
     }
     node {
      name=None
      _structType=&SpawnActorStateTreeTask
      bCanEditConsideredForCompletion=true
      bConsideredForCompletion=true
      bHasTaskCompletionDelegateDispatcher=false
      bTaskEnabled=true
      transitionHandlingPriority=Normal
      bindingsBatch {value=65535}
      executionRuntimeTemplateIndex {value=65535}
      instanceDataHandle {index=65535, source=None, stateHandle {index=65535}}
      instanceTemplateIndex {value=65535}
      outputBindingsBatch {value=65535}
     }
    }
   }]
  transitions [{id=#29, trigger=ON_STATE_COMPLETED, state {id=#30, name=Root, linkType=NEXT_STATE}}]
 }
 Spawn Slag 3 {
  id=#15
  type=STATE
  tasks [{
    id=#31
    description=Spawn Actor
    properties {
     iD=#32
     instance {
      _structType=&SpawnActorStateTreeTaskInstanceData
      actorClass=&B_Prop_SpireSpike_ShockWave_MinorSpire_C
      bAttachEncounterVolume=false
      bEnablePhysics=false
      bUseActorLocation=false
      bWaitForSpawnedActor=true
      owningActor=null
      persistenceBehavior=Destroy
      spawnedActor=null
      targetActor=null
      userActor=null
      actorIdentifierTags {gameplayTags [{tagName=QuestEncounter.Spire_Minor_Combat_1.Location.Object.3}]}
      actorTags {gameplayTags []}
      anchorOffset=<0 0 0>
      anchorQuery {
       autoDescription=" ANY( QuestEncounter.Spire_Minor_Combat_1.Location.Object.5, QuestEncounter.Spire_Minor_Combat_1.Location.Object.6 )"
       tokenStreamVersion=0
       userDescription=""
       queryTokenStream [0,1,1,2,0,1]
       tagDictionary [{tagName=QuestEncounter.Spire_Minor_Combat_1.Location.Object.5}, {tagName=QuestEncounter.Spire_Minor_Combat_1.Location.Object.6}]
      }
      existingActorQuery {autoDescription="", tokenStreamVersion=0, userDescription="", queryTokenStream [], tagDictionary []}
     }
     node {
      name=None
      _structType=&SpawnActorStateTreeTask
      bCanEditConsideredForCompletion=true
      bConsideredForCompletion=true
      bHasTaskCompletionDelegateDispatcher=false
      bTaskEnabled=true
      transitionHandlingPriority=Normal
      bindingsBatch {value=65535}
      executionRuntimeTemplateIndex {value=65535}
      instanceDataHandle {index=65535, source=None, stateHandle {index=65535}}
      instanceTemplateIndex {value=65535}
      outputBindingsBatch {value=65535}
     }
    }
   }]
  transitions [{id=#33, trigger=ON_STATE_COMPLETED, state {id=#34, name=Root, linkType=NEXT_STATE}}]
 }
 Grant Quest {
  id=#16
  type=STATE
  tasks [
   {
    id=#35
    description=Grant Quest
    properties {
     iD=#36
     instance {
      _structType=/Script/EncountersRuntime.GrantQuestStateTreeTaskInstanceData
      grantType=GrantToPlayspace
      questToGrant=/SpiresGameplay/Scenarios/Spire_Minor/Master_Scenario/SubScenarios/Spire_Minor_SubScenario_A/Quests/PlaySpace_EncounterQuest_Spire_SubScenario_A_01
      targetActor=null
      userActor=null
     }
     node {
      name=None
      _structType=/Script/EncountersRuntime.GrantQuestStateTreeTask
      bCanEditConsideredForCompletion=true
      bConsideredForCompletion=true
      bHasTaskCompletionDelegateDispatcher=false
      bTaskEnabled=false
      transitionHandlingPriority=Normal
      bindingsBatch {value=65535}
      executionRuntimeTemplateIndex {value=65535}
      instanceDataHandle {index=65535, source=None, stateHandle {index=65535}}
      instanceTemplateIndex {value=65535}
      outputBindingsBatch {value=65535}
     }
    }
   }
   {
    id=#37
    description=BP SDA STTask Complete State
    properties {
     iD=#38
     instanceObject=&self:StateTreeEditorData_0.StateTreeState_2.BP_SDA_STTask_CompleteState_C_0
     node {
      name=None
      _structType=&StateTreeBlueprintTaskWrapper
      taskClass=&BP_SDA_STTask_CompleteState_C
     }
    }
   }
  ]
  transitions [{id=#39, trigger=ON_STATE_COMPLETED, state {id=#40, name=Root, linkType=NEXT_STATE}}]
 }
 Detect Slag Destroyed {
  id=#17
  type=STATE
  tasks [{
    id=#41
    description=Scenario Verb Transition
    properties {
     iD=#42
     instance {
      _structType=/Script/EncountersRuntime.EncounterVerbTransitionStateTreeTaskInstanceData
      bIsTriggered=false
      bUseAdvancedObjectives=false
      objectiveValueThreshold=1
      userActor=null
      objectiveVerbs [{
        _structType=/Script/FortniteGame.FortObjectiveFilter_BuildingWasDestroyed
        author {
         bCanGainProgressForThisObjective=true
         tagQuery {
          autoDescription=" ANY( Building.Type.Prop.Slag )"
          tokenStreamVersion=0
          userDescription=""
          queryTokenStream [0,1,1,1,0]
          tagDictionary [{tagName=Building.Type.Prop.Slag}]
         }
        }
        context {tagQuery {autoDescription="", tokenStreamVersion=0, userDescription="", queryTokenStream [], tagDictionary []}}
        damageSource {
         bCanGainProgressForThisObjective=true
         tagQuery {autoDescription="", tokenStreamVersion=0, userDescription="", queryTokenStream [], tagDictionary []}
        }
       }]
      objectives []
      onTriggered {iD=#43}
     }
     node {
      name=None
      _structType=/Script/EncountersRuntime.EncounterVerbTransitionStateTreeTask
      bCanEditConsideredForCompletion=true
      bConsideredForCompletion=true
      bHasTaskCompletionDelegateDispatcher=false
      bTaskEnabled=true
      transitionHandlingPriority=Normal
      bindingsBatch {value=65535}
      executionRuntimeTemplateIndex {value=65535}
      instanceDataHandle {index=65535, source=None, stateHandle {index=65535}}
      instanceTemplateIndex {value=65535}
      outputBindingsBatch {value=65535}
      transitionTo {name=On Slag Destroyed, fallback=None, iD=#44, linkType=GotoState, stateHandle {index=65535}}
     }
    }
   }]
  transitions [
   {
    id=#45
    trigger=ON_STATE_COMPLETED
    conditions [{
      id=#46
      description="Make Int - Current Objective Count.Out Int.Int == Make Int - Max Objective Count.Out Int.Int"
      properties {
       iD=#47
       instance {_structType=/Script/StateTreeModule.StateTreeCompareIntConditionInstanceData, left=0, right=0}
       node {
        name=None
        _structType=/Script/StateTreeModule.StateTreeCompareIntCondition
        bInvert=false
        deltaIndent=0
        evaluationMode=Evaluated
        operand=And
        operator=Equal
        bindingsBatch {value=65535}
        executionRuntimeTemplateIndex {value=65535}
        instanceDataHandle {index=65535, source=None, stateHandle {index=65535}}
        instanceTemplateIndex {value=65535}
        outputBindingsBatch {value=65535}
       }
      }
     }]
    state {id=#19, name=Completed}
   }
   {id=#48, trigger=ON_STATE_COMPLETED, state {id=#17, name=Detect Slag Destroyed}}
  ]
 }
 On Slag Destroyed {
  id=#18
  type=STATE
  tasks [
   {
    id=#49
    description="(Fortnite) Arithmetic: Operation Int"
    properties {
     iD=#50
     instance {_structType=/Script/FortniteAIServer.FortAthenaIntArithmeticStateTreeTaskInstanceData, firstOperand=0, referencedResult="", secondOperand=1}
     node {
      name=None
      _structType=/Script/FortniteAIServer.FortAthenaIntArithmeticStateTreeTask
      bCanEditConsideredForCompletion=true
      bConsideredForCompletion=true
      bHasTaskCompletionDelegateDispatcher=false
      bTaskEnabled=true
      operation=Add
      operationTrigger=OnEnterState
      transitionHandlingPriority=Normal
      bindingsBatch {value=65535}
      executionRuntimeTemplateIndex {value=65535}
      instanceDataHandle {index=65535, source=None, stateHandle {index=65535}}
      instanceTemplateIndex {value=65535}
      outputBindingsBatch {value=65535}
     }
    }
   }
   {
    id=#51
    description=BP SDA STTask Complete State
    properties {
     iD=#52
     instanceObject=&self:StateTreeEditorData_0.StateTreeState_7.BP_SDA_STTask_CompleteState_C_0
     node {
      name=None
      _structType=&StateTreeBlueprintTaskWrapper
      taskClass=&BP_SDA_STTask_CompleteState_C
     }
    }
   }
  ]
  transitions [
   {
    id=#53
    trigger=ON_STATE_COMPLETED
    conditions [{
      id=#54
      description="Make Int - Current Objective Count.Out Int.Int == Make Int - Max Objective Count.Out Int.Int"
      properties {
       iD=#55
       instance {_structType=&StateTreeCompareIntConditionInstanceData, left=0, right=0}
       node {
        name=None
        _structType=&StateTreeCompareIntCondition
        bInvert=false
        deltaIndent=0
        evaluationMode=Evaluated
        operand=And
        operator=Equal
        bindingsBatch {value=65535}
        executionRuntimeTemplateIndex {value=65535}
        instanceDataHandle {index=65535, source=None, stateHandle {index=65535}}
        instanceTemplateIndex {value=65535}
        outputBindingsBatch {value=65535}
       }
      }
     }]
    state {id=#19, name=Completed}
   }
   {id=#56, trigger=ON_STATE_COMPLETED, state {id=#17, name=Detect Slag Destroyed}}
  ]
 }
 Completed {
  id=#19
  type=STATE
  tasks [
   {
    id=#57
    description=Send Custom Verb
    properties {
     iD=#58
     instance {
      _structType=/Script/EncountersRuntime.EncounterSendCustomVerbStateTreeTaskInstanceData
      bSendVerbForEachPlayerAuthor=true
      userActor=null
      customVerb {amount=1, customVerbTag {tagName=CustomVerb.Minor_Spire.Shield.Interact}, subjects [], values []}
     }
     node {
      name=None
      _structType=/Script/EncountersRuntime.EncounterSendCustomVerbStateTreeTask
      bCanEditConsideredForCompletion=true
      bConsideredForCompletion=true
      bHasTaskCompletionDelegateDispatcher=false
      bTaskEnabled=true
      transitionHandlingPriority=Normal
      bindingsBatch {value=65535}
      executionRuntimeTemplateIndex {value=65535}
      instanceDataHandle {index=65535, source=None, stateHandle {index=65535}}
      instanceTemplateIndex {value=65535}
      outputBindingsBatch {value=65535}
     }
    }
   }
   {
    id=#59
    description=BP SDA STTask Complete State
    properties {
     iD=#60
     instanceObject=&self:StateTreeEditorData_0.StateTreeState_8.BP_SDA_STTask_CompleteState_C_0
     node {
      name=None
      _structType=&StateTreeBlueprintTaskWrapper
      taskClass=&BP_SDA_STTask_CompleteState_C
     }
    }
   }
  ]
  transitions [{id=#61, trigger=ON_STATE_COMPLETED, state {id=#23, name=Root, linkType=SUCCEEDED}}]
 }
]
